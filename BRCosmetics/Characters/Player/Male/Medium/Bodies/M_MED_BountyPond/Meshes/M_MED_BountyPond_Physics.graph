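# Physics asset for the BountyPond medium male player body, set up for secondary motion of three shirt flaps.
# Each flap is a capsule chain of three or four dyn bones hanging from a base bone, linked by tight swing limits and zero twist limit.
# The base bones join spine_01 through free angular constraints.
# Pelvis, spine_01 and right thigh capsules supply body volumes; the left thigh has none.
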PhysicsAsset /BRCosmetics/Characters/Player/Male/Medium/Bodies/M_MED_BountyPond/Meshes/M_MED_BountyPond_Physics
bodies {
 base_shirt_flap_01 [{shapeType=Capsule, length=0.505, radius=0.505, shapeName=base_shirt_flap_01_capsule, rotation=<rot 0 -90 0>}]
 base_shirt_flap_02 [{shapeType=Capsule, length=0.505, radius=0.505, shapeName=base_shirt_flap_02_capsule, rotation=<rot 0 -90 0>}]
 base_shirt_flap_03 [{shapeType=Capsule, length=2, radius=2, shapeName=base_shirt_flap_03_capsule}]
 dyn_shirt_flap_01_01 [{shapeType=Capsule, length=2, radius=2, shapeName=dyn_shirt_flap_01_01_capsule, center=<1 0 0>}]
 dyn_shirt_flap_01_02 [{shapeType=Capsule, length=1, radius=3, shapeName=dyn_shirt_flap_01_02_capsule, center=<1 -0.8 0>}]
 dyn_shirt_flap_01_03 [{shapeType=Capsule, length=5, radius=3.5, shapeName=dyn_shirt_flap_01_03_capsule, center=<1 -2.9849 0>}]
 dyn_shirt_flap_01_04 [{shapeType=Capsule, length=0, radius=4, shapeName=dyn_shirt_flap_01_04_capsule, center=<1 -3.6191 0>}]
 dyn_shirt_flap_02_01 [{shapeType=Capsule, length=2, radius=2, shapeName=dyn_shirt_flap_02_01_capsule, center=<1 0 0>}]
 dyn_shirt_flap_02_02 [{shapeType=Capsule, length=0.5, radius=3.5, shapeName=dyn_shirt_flap_02_02_capsule, center=<2.2052 -0.8 0>}]
 dyn_shirt_flap_02_03 [{shapeType=Capsule, length=0.5, radius=3.5, shapeName=dyn_shirt_flap_02_03_capsule, center=<1 -1.6795 0>}]
 dyn_shirt_flap_02_04 [{shapeType=Capsule, length=0.5, radius=3.5, shapeName=dyn_shirt_flap_02_04_capsule, center=<1 -1.721 0>}]
 dyn_shirt_flap_03_01 [{shapeType=Capsule, length=2, radius=2, shapeName=dyn_shirt_flap_03_01_capsule, center=<1 0 0>}]
 dyn_shirt_flap_03_02 [{shapeType=Capsule, length=2, radius=2, shapeName=dyn_shirt_flap_03_02_capsule, center=<1 -1 0>}]
 dyn_shirt_flap_03_03 [{shapeType=Capsule, length=4, radius=3, shapeName=dyn_shirt_flap_03_03_capsule, center=<3 -1.9265 0>, rotation=<rot 90 0 0>}]
 pelvis [{shapeType=Capsule, length=12.8, radius=8, shapeName=pelvis_capsule, center=<-2.234 -0.6352 -0.5379>}]
 spine_01 [{shapeType=Capsule, length=18, radius=6, shapeName=spine_01_capsule, rotation=<rot 0 -90 0>}]
 thigh_r [{shapeType=Capsule, length=19.271, radius=6.0631, shapeName=thigh_r_capsule, center=<10.6834 0.7502 -1.3888>, rotation=<rot 90 0 0>}]
}
constraints [
 {bone1Name=dyn_shirt_flap_03_01, bone2Name=base_shirt_flap_03, swing1LimitDegrees=10, swing2LimitDegrees=5, twistLimitDegrees=0}
 {bone1Name=dyn_shirt_flap_03_02, bone2Name=dyn_shirt_flap_03_01, swing1LimitDegrees=10, swing2LimitDegrees=5, twistLimitDegrees=0}
 {bone1Name=dyn_shirt_flap_03_03, bone2Name=dyn_shirt_flap_03_02, swing1LimitDegrees=10, swing2LimitDegrees=5, twistLimitDegrees=0}
 {bone1Name=dyn_shirt_flap_02_04, bone2Name=dyn_shirt_flap_02_03, swing1LimitDegrees=10, swing2LimitDegrees=5, twistLimitDegrees=0}
 {bone1Name=dyn_shirt_flap_02_03, bone2Name=dyn_shirt_flap_02_02, swing1LimitDegrees=10, swing2LimitDegrees=5, twistLimitDegrees=0}
 {bone1Name=dyn_shirt_flap_02_02, bone2Name=dyn_shirt_flap_02_01, swing1LimitDegrees=10, swing2LimitDegrees=5, twistLimitDegrees=0}
 {bone1Name=dyn_shirt_flap_02_01, bone2Name=base_shirt_flap_02, swing1LimitDegrees=10, swing2LimitDegrees=5, twistLimitDegrees=0}
 {bone1Name=dyn_shirt_flap_01_04, bone2Name=dyn_shirt_flap_01_03, swing1LimitDegrees=10, swing2LimitDegrees=5, twistLimitDegrees=0}
 {bone1Name=dyn_shirt_flap_01_03, bone2Name=dyn_shirt_flap_01_02, swing1LimitDegrees=10, swing2LimitDegrees=5, twistLimitDegrees=0}
 {bone1Name=dyn_shirt_flap_01_02, bone2Name=dyn_shirt_flap_01_01, swing1LimitDegrees=10, swing2LimitDegrees=5, twistLimitDegrees=0}
 {bone1Name=dyn_shirt_flap_01_01, bone2Name=base_shirt_flap_01, swing1LimitDegrees=10, swing2LimitDegrees=5, twistLimitDegrees=0}
 {
  bone1Name=base_shirt_flap_03
  bone2Name=spine_01
  swing1LimitDegrees=45
  swing1Motion=Free
  swing2LimitDegrees=45
  swing2Motion=Free
  twistLimitDegrees=45
  twistMotion=Free
 }
 {
  bone1Name=base_shirt_flap_02
  bone2Name=spine_01
  swing1LimitDegrees=45
  swing1Motion=Free
  swing2LimitDegrees=45
  swing2Motion=Free
  twistLimitDegrees=45
  twistMotion=Free
 }
 {
  bone1Name=base_shirt_flap_01
  bone2Name=spine_01
  swing1LimitDegrees=45
  swing1Motion=Free
  swing2LimitDegrees=45
  swing2Motion=Free
  twistLimitDegrees=45
  twistMotion=Free
 }
]
