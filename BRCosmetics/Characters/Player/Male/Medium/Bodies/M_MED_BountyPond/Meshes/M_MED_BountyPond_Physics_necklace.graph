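# Physics asset for the necklace on the M_MED_BountyPond male player body, giving the pendant chain limited swinging motion.
# Two capsule bodies form a chain from a small base_necklace_01 capsule through dyn_necklace_01 to dyn_necklace_02.
# Both joints allow moderate swing on one axis, very little on the other and zero twist.
# A rotated box on spine_05, still named as a capsule, supplies a torso shape alongside the chain.
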PhysicsAsset /BRCosmetics/Characters/Player/Male/Medium/Bodies/M_MED_BountyPond/Meshes/M_MED_BountyPond_Physics_necklace
bodies {
 base_necklace_01 [{shapeType=Capsule, length=0.505, radius=0.505, shapeName=base_necklace_01_capsule, rotation=<rot 0 -90 0>}]
 dyn_necklace_01 [{shapeType=Capsule, length=5, radius=0.5, shapeName=dyn_necklace_01_capsule, center=<2.2 -1 0>}]
 dyn_necklace_02 [{shapeType=Capsule, length=5, radius=1, shapeName=dyn_necklace_02_capsule}]
 spine_05 [{shapeType=Box, extentX=9.5, extentY=17.5, extentZ=17.5, shapeName=spine_05_capsule, center=<1.7529 2.272 0>, rotation=<rot 0 54.0181 0>}]
}
constraints [
 {bone1Name=dyn_necklace_02, bone2Name=dyn_necklace_01, swing1LimitDegrees=25, swing2LimitDegrees=2, twistLimitDegrees=0}
 {bone1Name=dyn_necklace_01, bone2Name=base_necklace_01, swing1LimitDegrees=25, swing2LimitDegrees=2, twistLimitDegrees=0}
]
