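# Animation Blueprint for the BountyPond male medium body cosmetic, adding secondary motion to the pose copied from the parent mesh.
# The copied pose runs through Control Rig, additive backpack and necklace bone offsets, then two RigidBody simulations based at pelvis and spine_05.
# The update event computes gravity overrides from a parameter asset, with a skydive-float branch.
# It also derives a backpack rotation from spine yaw, varied by skydiving state.
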AnimBlueprint /BRCosmetics/Characters/Player/Male/Medium/Bodies/M_MED_BountyPond/Meshes/M_MED_BountyPond_AnimBP
target_skeleton=&/M_MED_BountyPond_Skeleton
graphs [
 AnimGraph {nodes [
   n1 = Control Rig!(Source=Copy Pose From Mesh() {bUseAttachedParent=true}, Alpha, IsCrouching:bool=Get bIsCrouching()) -> (Pose) {controlRigAssetReference {blueprintRigClass=&/M_MED_BountyPond_CtrlRig.M_MED_BountyPond_CtrlRig_C}}
   n2 = "Transform (Modify) Bone - Bone: attach_backpack"(ComponentPose=Local To Component(LocalPose=@n1), Alpha=1, Translation="0,-1.55,-4.067", Rotation="0.465,-3.988,7.858", Scale="1,1,1") {
    rotationMode=BMM_Additive
    rotationSpace=BCS_BoneSpace
    translationMode=BMM_Additive
    translationSpace=BCS_BoneSpace
    boneToModify {boneName=attach_backpack}
    rotation=<rot 0.465 -3.988 7.858>
    translation {y=-1.55, z=-4.067}
   }
   n3 = "Transform (Modify) Bone - Bone: attach_backpack"(ComponentPose=@n2, Alpha=1, Translation="0,0,0", Rotation="0,0,7", Scale="1,1,1") {
    rotationMode=BMM_Additive
    rotationSpace=BCS_BoneSpace
    translationMode=BMM_Additive
    translationSpace=BCS_BoneSpace
    boneToModify {boneName=attach_backpack}
    rotation {roll=7}
   }
   n4 = "Transform (Modify) Bone - Bone: base_necklace_01"!(ComponentPose=@n3, bAlphaBoolEnabled:bool=True, Translation="0,0,0", Rotation="0,-2.274,0", Scale="1,1,1") -> (Pose) {
    alphaInputType=Bool
    rotationMode=BMM_Additive
    rotationSpace=BCS_BoneSpace
    translationMode=BMM_Additive
    translationSpace=BCS_BoneSpace
    boneToModify {boneName=base_necklace_01}
    rotation {yaw=-2.274}
   }
   n5 = RigidBody!(ComponentPose=@n4, bAlphaBoolEnabled:bool, OverrideWorldGravity:Vector="0,0,0") -> (Pose) {
    alphaInputType=Bool
    bClampLinearTranslationLimitToRefPose=true
    bOverrideWorldGravity=true
    overridePhysicsAsset=&/M_MED_BountyPond_Physics
    simulationSpace=BaseBoneSpace
    baseBoneRef {boneName=pelvis}
    simSpaceSettings {
     dampingAlpha=0
     maxAngularAcceleration=45
     maxAngularVelocity=8
     maxLinearAcceleration=3500
     maxLinearVelocity=700
     velocityScaleZ=0.5
     worldAlpha=1
     externalAngularVelocity=<0.3 0.3 0.3>
     externalLinearVelocity=<0.3 0.3 0.3>
    }
   }
   n6 = RigidBody(ComponentPose=@n5, OverrideWorldGravity="0,0,0") {
    alphaInputType=Bool
    bClampLinearTranslationLimitToRefPose=true
    bOverrideWorldGravity=true
    overridePhysicsAsset=&/M_MED_BountyPond_Physics_necklace
    simulationSpace=BaseBoneSpace
    baseBoneRef {boneName=spine_05}
    simSpaceSettings {
     dampingAlpha=0
     maxAngularAcceleration=40
     maxAngularVelocity=8
     maxLinearAcceleration=800
     maxLinearVelocity=500
     velocityScaleZ=0.5
     worldAlpha=0.7
     externalAngularVelocity=<0.3 0.3 0.3>
     externalLinearVelocity=<0.3 0.3 0.3>
    }
   }
   Output Pose(Result=Component To Local(ComponentPose=@n6)) {layerGroup=None}
  ]}
 EventGraph {nodes [
   Event BlueprintUpdateAnimation()
   > n1 = Sequence!(execute) -> (then_0, then_1, then_2:Exec)
   > n2 = "Gravity Override State Function (Parameter Asset) (Nativized)"(
    CharacterPartAnimInstance=Self-Reference()
    GameTimeInSeconds=GetGameTimeInSeconds()
    Parameters=&/M_MED_BountyPond_Parameters
   )
   > n3 = Set Gravity Override(Gravity Override=@n2)
   > n4 = Set RBAN Settings Out(RBAN Settings Out=@n2.OutRigidBodyAnimNodeParameters)
   Set GravityOverride_cloth(execute=[@n11.else, @n4.then], GravityOverride_cloth:Vector=@n3) -> (Output_Get:Vector)
   n5 = Branch(execute=@n1.then_1, Condition=Get rotateback() -> (rotateback:bool))
   n6 = GetOwnerSkeletalMeshComponent()
   n7 = BreakRotator(InRot=BreakTransform(InTransform=GetDeltaTransformFromRefPose(self=@n6, BoneName=spine_05, BaseName=spine_01)).Rotation)
   n8 = MapRangeClamped(Value=@n7.Yaw, InRangeA=-10, InRangeB=-30, OutRangeB=-10)
   n9 = MapRangeClamped(Value=@n7.Yaw, InRangeA=-10, InRangeB=-30, OutRangeA=10, OutRangeB=20)
   n10 = MapRangeClamped(Value=@n7.Yaw, InRangeA=-10, InRangeB=-30, OutRangeB=10)
   > Set backpack_rotate(backpack_rotate="Scale Rotator (Float)"(A:Rotator=MakeRotator(Roll=@n9, Pitch=@n10, Yaw=@n8), B:float=0.3) -> (ReturnValue:Rotator=~) {is=title K2Node_PromotableOperator})
   Set backpack_rotate(execute=@n5.else, backpack_rotate=MakeRotator(Pitch=@n10, Yaw=@n8))
   Event OnBeginSkydiving()
   > Set rotateback(rotateback:bool=true) -> (Output_Get:bool)
   Event OnEndSkydiving()
   > Set rotateback(rotateback=false)
   n11 = Branch(Condition=Get bIsSkydiveFloating())
   > Set GravityOverride_cloth(GravityOverride_cloth="0,0,-1200")
  ]}
]
variables [
 backpack_rotate {type=Rotator}
 Gravity Override {type=Vector}
 GravityOverride_cloth {type=Vector}
 RBAN Settings Out {type=FortRigidBodyAnimNodeParameters}
 rotateback {type=boolean}
]
